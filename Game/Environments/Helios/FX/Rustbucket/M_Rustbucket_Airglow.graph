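# Additive, unlit, two-sided Niagara sprite material for an airglow effect, filed under the Rustbucket FX of the Helios environment.
# A UV-centered SphereMask forms a round falloff, with radius from a dynamic parameter and hardness and exponent from scalar parameters.
# DepthFade softens it over a per-particle distance, then particle alpha multiplies it.
# An orange vector parameter and a spare scalar feed no node shown.
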Material /Game/Environments/Helios/FX/Rustbucket/M_Rustbucket_Airglow
graph {
 properties {
  bAutomaticallySetUsageInEditor=false
  bUsedWithNiagaraSprites=true
  blendMode=BLEND_Additive
  pixelDepthOffsetMode=PDOM_Legacy
  shadingModel=MSM_Unlit
  twoSided=true
  usedShadingModels=Unlit
 }
 nodes [
  n1 = DynamicParameter()
  n2 = SphereMask(A=TextureCoordinate(), B=Constant() {r=0.5}, Radius=@n1.Param1, Hardness=ScalarParameter() {parameterName=Param_3})
  n3 = ScalarParameter() {defaultValue=2, parameterName=Param_4}
  Multiply(A=ParticleColor().A, B=DepthFade(Opacity=Power(Base=@n2, Exp=@n3), FadeDistance=@n1.Param2))
  ScalarParameter() {defaultValue=0.5, parameterName=Param_2}
  VectorParameter() {defaultValue=<rgba 2.5 0.6349 0.151 1>}
 ]
}
