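# Unlit translucent two-sided material for Rustbucket ember effects, animating drifting, glowing particles entirely in the shader.
# Vertex color varies each ember's direction, looping time-based travel, size and color, while world position offset moves and billboards quads through the Sprite function.
# Low-to-high ember colors pass through hue shift and HDR boosting.
# Soft circular and lifetime masks combine with camera-distance and near-camera fades, and distance also adjusts sprite size.
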Material /Game/Environments/Helios/FX/Rustbucket/M_Rustbucket_Embers
graph {
 properties {
  bAutomaticallySetUsageInEditor=false
  blendMode=BLEND_Translucent
  pixelDepthOffsetMode=PDOM_Legacy
  shadingModel=MSM_Unlit
  twoSided=true
  usedShadingModels=Unlit
 }
 nodes [
  n1 = VectorParameter() {defaultValue=<rgba -50 0 75 0>}
  n2 = VertexColor()
  n3 = Multiply(A=ConstantBiasScale(None=@n2) {bias=-0.5, scale=2}, B=ScalarParameter() {defaultValue=75, parameterName=Param_4})
  n4 = ScalarParameter() {defaultValue=0.5, parameterName=Param_13}
  n5 = Time()
  n6 = Fmod(
   A=Add(A=@n2.A, B=Multiply(A=@n4, B=@n5))
   B=LinearInterpolate(A=ScalarParameter() {defaultValue=1, parameterName=Param_7}, B=ScalarParameter() {defaultValue=5, parameterName=Param_5}, Alpha=@n2.G)
  )
  n7 = PixelDepth()
  n8 = ScalarParameter() {defaultValue=1500, parameterName=Camera Distance}
  n9 = LinearInterpolate(
   A=ScalarParameter() {defaultValue=0.5, parameterName=Param_8}
   B=ScalarParameter() {parameterName=Param_14}
   Alpha=Saturate(None=Divide(A=@n7.R, B=@n8))
  ) {constB=2}
  n10 = Multiply(A=ScalarParameter() {defaultValue=1, parameterName=Param_1}, B=@n9)
  n11 = Multiply(A=ScalarParameter() {defaultValue=2, parameterName=Param_2}, B=@n9)
  n12 = MaterialFunctionCall(X=LinearInterpolate(A=@n10, B=@n11, Alpha=@n2.G), Y=LinearInterpolate(A=@n10, B=@n11, Alpha=@n2.G)) {materialFunction=&MakeFloat2}
  n13 = Step(Y=@n2.A, X=ScalarParameter() {defaultValue=0.12, parameterName=Param_12})
  Add(
   A=Multiply(A=Add(A=@n1.RGB, B=@n3), B=@n6)
   B=MaterialFunctionCall!(
    Center Location=WorldPosition().XYZ
    0-1 UVs=TextureCoordinate()
    XY scale=Multiply(A=@n12.Result, B=@n13)
    OverridePivot
    "Normals (Optional)"
   ) {materialFunction=&Sprite}.World position offset
  )
  LinearInterpolate()
  n14 = ScalarParameter() {defaultValue=1, parameterName=Param_3}
  LinearInterpolate(B=Multiply(B=Fmod(B=@n14)))
  n15 = ScalarParameter() {parameterName="Hue Shift Percentage (S)"}
  n16 = LinearInterpolate(
   A=VectorParameter() {parameterName=Ember Low, defaultValue=<rgba 1 0.1011 0 1>}.RGB
   B=VectorParameter() {parameterName=Ember High, defaultValue=<rgba 1 0.5607 0 1>}.RGB
   Alpha=@n2
  )
  MaterialFunctionCall(
   LDR Input=MaterialFunctionCall(Hue Shift Percentage=@n15, Texture=Saturate(None=@n16)) {materialFunction=&HueShift}.Result
   HDR Tint=MaterialFunctionCall(Hue Shift Percentage=@n15, Texture=@n16) {materialFunction=&HueShift}.Result
   HDR Intensity Multiplier=ScalarParameter() {defaultValue=350, parameterName=Param_6}
  ) {materialFunction=&DeriveHDRfromLDR}
  n17 = Saturate(None=Divide(A=PixelDepth().R, B=ScalarParameter() {defaultValue=1500, parameterName=Camera Distance}))
  n18 = SphereMask(A=@n6, B=ScalarParameter() {parameterName=Param_10}, Radius=ScalarParameter() {defaultValue=1, parameterName=Param_11}) {hardnessPercent=0}
  n19 = SphereMask(
   A=TextureCoordinate()
   B=Constant() {r=0.5}
   Radius=ScalarParameter() {defaultValue=0.5}
   Hardness=ScalarParameter() {defaultValue=0.25, parameterName=Param_9}
  ) {hardnessPercent=0}
  n20 = ScalarParameter() {defaultValue=800, parameterName=NearFade}
  n21 = ScalarParameter() {defaultValue=250, parameterName=NearFadeGone}
  Multiply(
   A=Multiply(A=Multiply(A=@n18, B=@n19), B=LinearInterpolate(Alpha=@n17))
   B=MaterialFunctionCall(Distance=Subtract(A=@n20, B=@n21)) {materialFunction=&Near_Camera_Fade}.Result
  )
 ]
}
