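# Additive, unlit, two-sided flame material for the Helios Rustbucket effects, animating mesh cards as looping fire sprites entirely in the shader.
# World position offset billboards cards through the Sprite function, with vertex color randomizing size, culling and loop phase, plus wind drift and camera offset.
# Panned noise, distorted by a second flame texture, is shaped by UV gradients and lifetime, depth and camera fades.
# Emissive color uses hue-shift and LDR-to-HDR functions.
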Material /Game/Environments/Helios/FX/Rustbucket/M_Rustbucket_Flames
graph {
 properties {
  bAutomaticallySetUsageInEditor=false
  blendMode=BLEND_Additive
  pixelDepthOffsetMode=PDOM_Legacy
  shadingModel=MSM_Unlit
  twoSided=true
  usedShadingModels=Unlit
 }
 nodes [
  n1 = Saturate(None=DotProduct(
    A=LocalPosition().XYZ
    B=MaterialFunctionCall!(Use World Position Offsets) {materialFunction=/Engine/Functions/Engine_MaterialFunctions02/Utility/CameraVectorWithWPOOptions}.Result
   ))
  n2 = VertexColor()
  n3 = OneMinus(None=ComponentMask(None=TextureCoordinate()) {g=true})
  n4 = MaterialFunctionCall(
   X=LinearInterpolate(
    A=ScalarParameter() {defaultValue=15, parameterName=Sprite Size X Min}
    B=ScalarParameter() {defaultValue=30, parameterName=Sprite Size X Max}
    Alpha=@n2.G
   ) {constB=0}
   Y=LinearInterpolate(
    A=ScalarParameter() {parameterName=Sprite Size Y - Min}
    B=ScalarParameter() {defaultValue=30, parameterName=Sprite Size Y - Max}
    Alpha=Multiply(A=@n2.G, B=@n3)
   )
  ) {materialFunction=&MakeFloat2}
  n5 = VertexColor()
  n6 = ScalarParameter() {defaultValue=0.5, parameterName=Spawn Rate}
  n7 = MaterialFunctionCall!(
   Center Location=TransformPosition(None=LocalPosition().XYZ) {transformType=TRANSFORMPOSSOURCE_World}
   0-1 UVs=TextureCoordinate()
   XY scale=Multiply(A=Multiply(A=@n1, B=@n4.Result), B=Step(Y=@n5.A, X=@n6))
   OverridePivot=StaticBoolParameter() {defaultValue=true, parameterName="OverridePivot (SB)"}
   "Normals (Optional)"
  ) {materialFunction=&Sprite}
  n8 = NamedRerouteUsage()
  n9 = VectorParameter() {parameterName=Pivot Offset, defaultValue=<rgba 0 0 -9 0>}
  n10 = Time()
  n11 = ScalarParameter() {defaultValue=1, parameterName=Time Scale}
  n12 = Fmod(A=Add(A=@n5.A, B=Multiply(A=@n10, B=@n11)), B=ScalarParameter() {defaultValue=1, parameterName=Loop Duration})
  n13 = ComponentMask(None=TextureCoordinate()) {g=true}
  n14 = Multiply(A=@n12, B=LinearInterpolate(B=VectorParameter() {parameterName=Wind Direction, defaultValue=<rgba 0 0 50 0>}.RGB, Alpha=OneMinus(None=@n13)))
  Add(
   A=Add(A=Multiply(A=@n7.World position offset, B=@n8), B=Add(A=@n9.RGB, B=@n14))
   B=MaterialFunctionCall(Offset Amount=ScalarParameter() {defaultValue=10, parameterName=Camera Offset}, World Position=WorldPosition().XYZ) {materialFunction=/Engine/Functions/Engine_MaterialFunctions02/WorldPositionOffset/CameraOffset}.Camera Offset
  )
  n15 = TextureCoordinate()
  LinearInterpolate(A=VectorParameter() {parameterName=asdf}.RGB, Alpha=OneMinus(None=ComponentMask(None=@n15) {g=true}))
  n16 = ScalarParameter() {parameterName="Hue Shift Percentage (S)"}
  n17 = VectorParameter() {parameterName=Emissive Color, defaultValue=<rgba 5 0.6285 0.25 1>}
  n18 = Saturate(None=@n17.RGB)
  MaterialFunctionCall(Emissive=MaterialFunctionCall(
    LDR Input=MaterialFunctionCall(Hue Shift Percentage=@n16, Texture=@n18) {materialFunction=&HueShift}.Result
    HDR Tint=MaterialFunctionCall(Hue Shift Percentage=@n17.RGB, Texture=@n16) {materialFunction=&HueShift}.Result
    HDR Intensity Multiplier=ScalarParameter() {defaultValue=2.5, parameterName=HDR Intensity}
   ) {materialFunction=&DeriveHDRfromLDR}.RemappedRange) {materialFunction=&MF_FX_EmissiveAdjustments}
  Multiply(A=@n1)
  n19 = TextureCoordinate()
  n20 = VertexColor()
  n21 = Multiply(
   A=ComponentMask(None=VectorParameter() {parameterName=UV Scale Main, defaultValue=<rgba 0.5 0.1 0 0>}.RGB) {g=true, r=true}
   B=LinearInterpolate(Alpha=@n20.G)
  )
  n22 = Time()
  n23 = LinearInterpolate(
   A=ScalarParameter() {defaultValue=0.12, parameterName=Random Noise Pan Min}
   B=ScalarParameter() {defaultValue=0.5, parameterName=Random Pan Noise Max}
   Alpha=@n20.B
  ) {constA=0.5}
  n24 = Add(A=@n20.G, B=TextureCoordinate())
  n25 = ComponentMask(None=VectorParameter() {parameterName=Distortion UV Scale, defaultValue=<rgba 0.25 0.12 0 0>}.RGB) {g=true, r=true}
  n26 = Time()
  n27 = ScalarParameter() {defaultValue=0.25, parameterName=Distortion Pan Speed}
  n28 = TextureSampleParameter2D(UVs=Panner(Coordinate=Multiply(A=@n24, B=@n25), Time=Multiply(A=@n26, B=@n27)) {speedY=1}) {
   parameterName=FlameTexture
   samplerType=SAMPLERTYPE_LinearGrayscale
   texture=/Game/Weapons/FORT_Melee/Pickaxe_Cube_Ninja_Male/FX/T_FluidFlameTile
  }
  n29 = Add(
   A=Panner(Coordinate=Multiply(A=@n19, B=@n21), Time=Multiply(A=@n22, B=@n23)) {speedY=1}
   B=Multiply(A=ComponentMask(None=@n28.RGB) {g=true, r=true}, B=ScalarParameter() {defaultValue=0.5, parameterName=Flame UV Distortion})
  )
  n30 = ComponentMask(None=Add(A=TextureCoordinate(), B=TextureSample(UVs=@n29) {texture=/Game/Global/Textures/Default/Noise/T_FireNoise_02}.G)) {g=true}
  n31 = ScalarParameter() {defaultValue=15, parameterName=Flame Contrast}
  n32 = ComponentMask(None=TextureCoordinate()) {r=true}
  n33 = Multiply(A=OneMinus(None=@n32), B=@n32)
  n34 = ScalarParameter() {defaultValue=6, parameterName=UV Top and Bottom Gradient Power}
  n35 = Multiply(
   A=Multiply(A=Power(Base=@n30, Exp=@n31), B=Power(Base=@n33, Exp=@n34))
   B=LinearInterpolate(B=ScalarParameter() {defaultValue=10, parameterName=Random Brightness Mult}, Alpha=VertexColor().G) {constA=1}
  )
  n36 = SphereMask(
   A=@n12
   B=ScalarParameter() {parameterName=Fade Start}
   Radius=ScalarParameter() {defaultValue=1, parameterName=Fade Scale over Life}
   Hardness=ScalarParameter() {defaultValue=0.75, parameterName=Fade Scale Hardness}
  )
  n37 = OneMinus(None=ComponentMask(None=TextureCoordinate()) {g=true})
  n38 = ScalarParameter() {defaultValue=1.5, parameterName=UV Bottom Gradient}
  n39 = Multiply(A=Multiply(A=Multiply(A=@n35, B=@n36), B=Power(Base=@n37, Exp=@n38)), B=NamedRerouteUsage())
  n40 = ScalarParameter() {defaultValue=10, parameterName=Depth Fade Distance}
  n41 = Subtract(A=ScalarParameter() {defaultValue=800, parameterName=NearFade}, B=ScalarParameter() {defaultValue=250, parameterName=NearFadeGone})
  Multiply(
   A=Multiply(
    A=DepthFade(Opacity=@n39, FadeDistance=@n40)
    B=MaterialFunctionCall(Distance=@n41) {materialFunction=&Near_Camera_Fade}.Result
   )
   B=ScalarParameter() {defaultValue=1, parameterName=Overall Opacity Mult}
  )
  n42 = Divide(A=PixelDepth().R, B=ScalarParameter() {defaultValue=500, parameterName=Camera Distance})
  NamedRerouteDeclaration(None=StaticSwitchParameter(True=Saturate(None=@n42), False=Constant() {r=1}) {parameterName=Use Camera Distance Fade}) {name=Camera Fade Distance, nodeColor=<rgba 1 0 0.3984 1>}
  OneMinus()
  Transform() {transformSourceType=TRANSFORMSOURCE_World, transformType=TRANSFORM_Instance}
  WorldPosition()
  WorldPosition()
 ]
}
