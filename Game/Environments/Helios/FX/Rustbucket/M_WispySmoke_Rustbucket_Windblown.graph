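# Translucent, two-sided lit material for wispy, windblown smoke trails in the Helios Rustbucket environment FX.
# World position offset combines SplineThicken widening from base to tip, a UV-weighted bend toward a sine-and-cosine-oscillating wind vector, and scrolling curl-noise distortion.
# Two panning smoke textures, shaped by a UV gradient and DepthFade, form opacity.
# Their blended normals add a cylinder normal for roundness, replaced by constants on mobile and low or medium quality.
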Material /Game/Environments/Helios/FX/Rustbucket/M_WispySmoke_Rustbucket_Windblown
graph {
 properties {
  bAutomaticallySetUsageInEditor=false
  blendMode=BLEND_Translucent
  pixelDepthOffsetMode=PDOM_Legacy
  translucencyLightingMode=TLM_VolumetricPerVertexDirectional
  twoSided=true
  usedShadingModels=Default Lit
 }
 nodes [
  n1 = OneMinus(None=ComponentMask(None=TextureCoordinate()) {g=true})
  n2 = ScalarParameter() {defaultValue=15, parameterName=Curl Scale}
  n3 = WorldPosition()
  n4 = ScalarParameter() {defaultValue=0.0005, parameterName=WPO Noise Scale}
  n5 = Time()
  n6 = VectorParameter() {parameterName=WPO Curl Vector, defaultValue=<rgba 0 0 -0.1 0>}
  n7 = TextureSample(UVs=Add(A=Multiply(A=@n3.XYZ, B=@n4), B=Multiply(A=@n5, B=@n6.RGB))) {samplerType=SAMPLERTYPE_LinearColor, texture=/Game/Effects/FogVolume/VT_CurlNoise_LOW}
  n8 = Time()
  n9 = ScalarParameter() {defaultValue=0.25, parameterName=Sine Scale}
  n10 = ScalarParameter() {defaultValue=0.12, parameterName=Cosine Scale}
  n11 = Multiply(A=Sine(None=Multiply(A=@n8, B=@n9)), B=Cosine(None=Multiply(A=@n8, B=@n10)))
  n12 = LinearInterpolate(
   A=VectorParameter() {parameterName=Wind Direction Low, defaultValue=<rgba 300 0 -25 0>}.RGB
   B=VectorParameter() {parameterName=Wind DIrection High, defaultValue=<rgba 350 0 50 0>}.RGB
   Alpha=@n11
  )
  n13 = ScalarParameter() {defaultValue=0.1, parameterName=Param_3}
  n14 = Power(
   Base=@n1
   Exp=LinearInterpolate(
    A=ScalarParameter() {defaultValue=0.9, parameterName=Bend Power Low}
    B=ScalarParameter() {defaultValue=1.5, parameterName=Bend Power High}
    Alpha=Multiply(A=@n13, B=@n11)
   )
  )
  n15 = ScalarParameter() {defaultValue=12, parameterName=Width Base}
  n16 = ScalarParameter() {defaultValue=350, parameterName=Width Tip}
  Add(
   A=Multiply(A=@n1, B=Multiply(A=@n2, B=@n7.RGB))
   B=Add(
    A=LinearInterpolate(B=@n12, Alpha=@n14)
    B=MaterialFunctionCall(WidthBase=@n15, WidthTip=@n16) {materialFunction=&SplineThicken}.WorldPositionOffset
   )
  )
  n17 = Power(Base=ComponentMask(None=TextureCoordinate()) {g=true}, Exp=ScalarParameter() {defaultValue=2.5, parameterName=power gradient})
  n18 = TextureCoordinate()
  n19 = MaterialFunctionCall(X=ScalarParameter() {defaultValue=1, parameterName=Tiling 2 X}, Y=ScalarParameter() {defaultValue=3, parameterName=Tiling 2 Y}) {materialFunction=&MakeFloat2}
  n20 = Multiply(A=Time(), B=ScalarParameter() {defaultValue=1, parameterName=Global Panner Scale})
  n21 = MaterialFunctionCall(XY=Panner(Coordinate=Multiply(A=@n18, B=@n19.Result), Time=@n20) {speedY=0.75}) {materialFunction=&Swizzle}
  n22 = TextureSample(UVs=@n21.YX) {samplerType=SAMPLERTYPE_Grayscale, texture=&T_Smoke_Trail_Soft}
  n23 = TextureCoordinate()
  n24 = MaterialFunctionCall(X=ScalarParameter() {defaultValue=1, parameterName=Tiling X}, Y=ScalarParameter() {defaultValue=3, parameterName=Tiling Y}) {materialFunction=&MakeFloat2}
  n25 = MaterialFunctionCall(XY=Panner(Coordinate=Multiply(A=@n23, B=@n24.Result), Time=@n20) {speedY=1}) {materialFunction=&Swizzle}
  n26 = TextureSample(UVs=@n25.YX) {samplerType=SAMPLERTYPE_Grayscale, texture=&T_Smoke_Trail_Soft}
  n27 = Power(Base=Saturate(None=Add(A=@n22.G, B=@n26.G)), Exp=ScalarParameter() {defaultValue=1, parameterName=Texture Power})
  DepthFade(Opacity=Multiply(A=Multiply(A=@n17, B=@n27), B=ScalarParameter() {defaultValue=0.5, parameterName=Opacity Scale}))
  Multiply()
  n28 = MaterialFunctionCall(
   BaseNormal=TextureSample(UVs=@n25.YX) {samplerType=SAMPLERTYPE_Normal, texture=/Game/Effects/Fort_Effects/Textures/Smoke/T_Smoke_Trail_Soft_N}.RGB
   AdditionalNormal=TextureSample(UVs=@n21.YX) {samplerType=SAMPLERTYPE_Normal, texture=&T_Smoke_Trail_Soft_N}.RGB
  ) {materialFunction=&BlendAngleCorrectedNormals}
  n29 = TextureSample() {samplerType=SAMPLERTYPE_Normal, texture=/Game/Environments/Asteria/Foliage/Grass/Textures/Grass/CylinderNormal_N}
  n30 = Constant()
  QualitySwitch(
   Default=ShadingPathSwitch(
    Default=MaterialFunctionCall(BaseNormal=@n28.Result, AdditionalNormal=@n29.RGB) {materialFunction=&BlendAngleCorrectedNormals}.Result
    Mobile=Constant()
   )
   Low=@n30
   Medium=@n30
  )
  Transform() {transformSourceType=TRANSFORMSOURCE_World, transformType=TRANSFORM_Local}
  VectorParameter() {parameterName=Albedo, defaultValue=<rgba 0.1021 0.1021 0.1021 0>}
 ]
}
