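# Niagara system for forge ember effects in the Rustbucket part of the Helios environment.
# It combines three emitters, named Ember, Airglow and Core, suggesting sparks, a surrounding glow and a central element.
# The listed logic is only a SystemState module and a random float range input; no renderers, user variables or data interfaces appear.
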NiagaraSystem /Game/Environments/Helios/FX/Rustbucket/NS_RustBucket_Forge_Embers
dependencies {
 usedDataInterfaces []
 usedDynamicInputs [&RandomRangeFloat]
 usedModules [&SystemState]
}
emitters {Airglow {rendererClasses [], renderers []}, Core {rendererClasses [], renderers []}, Ember {rendererClasses [], renderers []}}
summary {
 systemName=NS_RustBucket_Forge_Embers
 emitters [{emitterName=Ember, rendererClasses []}, {emitterName=Airglow, rendererClasses []}, {emitterName=Core, rendererClasses []}]
 userVariables []
}
