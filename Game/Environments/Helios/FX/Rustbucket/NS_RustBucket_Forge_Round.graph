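# NS_RustBucket_Forge_Round is a Niagara environment effect for the Helios Rustbucket set, named for a round forge.
# It holds three emitters, Ember, Airglow and Airglow001, whose names suggest sparks and two glow layers.
# The file lists only the SystemState module, with no renderers, user variables, data interfaces or dynamic inputs, so the emitters' behavior and look are not shown.
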NiagaraSystem /Game/Environments/Helios/FX/Rustbucket/NS_RustBucket_Forge_Round
dependencies {usedDataInterfaces [], usedDynamicInputs [], usedModules [&SystemState]}
emitters {Airglow {rendererClasses [], renderers []}, Airglow001 {rendererClasses [], renderers []}, Ember {rendererClasses [], renderers []}}
summary {
 systemName=NS_RustBucket_Forge_Round
 emitters [{emitterName=Ember, rendererClasses []}, {emitterName=Airglow, rendererClasses []}, {emitterName=Airglow001, rendererClasses []}]
 userVariables []
}
