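# Niagara system for forge smoke in the Rustbucket area of the Helios environment, judging by its name and folder.
# It holds a single emitter named Steam, with no renderers listed.
# The only listed module is SystemState, and no user variables, data interfaces or dynamic inputs appear, so the particle behavior is not shown.
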NiagaraSystem /Game/Environments/Helios/FX/Rustbucket/NS_RustBucket_Forge_Smoke
dependencies {usedDataInterfaces [], usedDynamicInputs [], usedModules [&SystemState]}
emitters {Steam {rendererClasses [], renderers []}}
summary {systemName=NS_RustBucket_Forge_Smoke, emitters [{emitterName=Steam, rendererClasses []}], userVariables []}
